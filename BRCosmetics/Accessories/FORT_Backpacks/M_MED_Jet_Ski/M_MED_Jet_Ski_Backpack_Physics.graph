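# Physics asset for the Jet Ski backpack cosmetic, giving its dangling attachments physical secondary motion.
# Box bodies cover the pack and its dynamic bones: two grenade pieces, a radio, a handle and left and right straps.
# Each dynamic body is constrained directly to the pack in a star layout with narrow twist limits.
# The straps get wide swing ranges, while the handle, grenades and radio stay tightly limited.
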PhysicsAsset /BRCosmetics/Accessories/FORT_Backpacks/M_MED_Jet_Ski/M_MED_Jet_Ski_Backpack_Physics
bodies {
 dyn_grenade_b [{shapeType=Box, extentX=3.1851, extentY=2.3635, extentZ=7.0216, center=<1.1984 0.2082 0.0589>}]
 dyn_grenades [{shapeType=Box, extentX=3.1014, extentY=2.7512, extentZ=7.0216, center=<1.1984 0.2082 0.0589>}]
 dyn_handle [{shapeType=Box, extentX=1.5352, extentY=4.1343, extentZ=10.7681, center=<3.8202 1.2393 -0.275>}]
 dyn_radio [{shapeType=Box, extentX=5.736, extentY=8.7303, extentZ=11.7114, center=<3.23 -0.6965 -0.1708>}]
 dyn_strap_l [{shapeType=Box, extentX=7.3569, extentY=1.01, extentZ=3.6412, center=<3.4902 -0.1293 -0.1835>}]
 dyn_strap_r [{shapeType=Box, extentX=6.8008, extentY=1.01, extentZ=3.3216, center=<3.6871 -0.3519 0.0844>}]
 pack [{shapeType=Box, extentX=23.4725, extentY=9.9204, extentZ=26.647, center=<-0.6835 -3.9716 -4.5901>, rotation=<rot 0 0 -11.1499>}]
}
constraints [
 {bone1Name=dyn_grenades, bone2Name=pack, swing1LimitDegrees=10, swing2LimitDegrees=8, twistLimitDegrees=3}
 {bone1Name=dyn_radio, bone2Name=pack, swing1LimitDegrees=10, swing2LimitDegrees=15, twistLimitDegrees=2}
 {bone1Name=dyn_strap_l, bone2Name=pack, swing1LimitDegrees=125, swing2LimitDegrees=40, twistLimitDegrees=2}
 {bone1Name=dyn_strap_r, bone2Name=pack, swing1LimitDegrees=125, swing2LimitDegrees=40, twistLimitDegrees=2}
 {bone1Name=dyn_handle, bone2Name=pack, swing1LimitDegrees=5, swing2LimitDegrees=5, twistLimitDegrees=2}
 {bone1Name=dyn_grenade_b, bone2Name=pack, swing1LimitDegrees=12, swing2LimitDegrees=12, twistLimitDegrees=7}
]
